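# Material function for the Princess Castle set that adds a fuzzy sheen to a material's base color, named for sand.
# It passes the incoming base color through a shared world-space fuzzy shading function, with core darkness, power and edge brightness exposed as Fuzz parameters.
# A quality switch keeps the plain color on Low and Medium, and the material proxy also uses it.
# The result replaces only base color in the returned material attributes.
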MaterialFunction /Game/Environments/Sets/PrincessCastle/MaterialFunctions/MF_PrincessCastle_SandFuzz
graph {nodes [
  n1 = FunctionInput() {inputType=FunctionInput_MaterialAttributes}
  n2 = GetMaterialAttributes(None=@n1) {attributeGetTypes [#1]}
  n3 = QualitySwitch(
   Default=MaterialFunctionCall!(
    BaseColor=@n2.BaseColor
    Normal
    CoreDarkness=ScalarParameter() {group=Fuzz, parameterName=Fresnel Core Darkness}
    Power=ScalarParameter() {defaultValue=3.5, group=Fuzz, parameterName=Fresnel Core Power}
    EdgeBrightness=ScalarParameter() {defaultValue=0.75, group=Fuzz, parameterName=Fresnel Edge Brightness}
   ) {materialFunction=/Game/Packages/Fortress_SharedMaterials/Base_Material_Functions/FuzzyShadingWorldSpace}.Result
   Low=@n2.BaseColor
   Medium=@n2.BaseColor
  )
  FunctionOutput(None=MaterialFunctionCall!(Material=@n1, NewBaseColor=MaterialProxyReplace(Realtime=@n3, MaterialProxy=@n2.BaseColor), Mask) {materialFunction=/Engine/Functions/MaterialLayerFunctions/MatLayerBlend_OverrideBaseColor}.Blended Material)
 ]}
