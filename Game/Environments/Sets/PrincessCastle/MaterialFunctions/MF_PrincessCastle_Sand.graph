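# Material function that reskins Princess Castle environment surfaces as sparkling sand, behind the Use_SandCastle static switch.
# Triplanar projections at several scales supply tinted base color, blended normals, roughness, specular and AO.
# Glints come from a time-driven hue-shifted detail texture combined with view direction, sunlight and camera-distance fade, with emissive zeroed on mobile.
# The rebuilt material attributes pass through MF_PrincessCastle_SandFuzz before output.
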MaterialFunction /Game/Environments/Sets/PrincessCastle/MaterialFunctions/MF_PrincessCastle_Sand
graph {nodes [
  n1 = CollectionParameter() {
   collection=&FortniteMaterialParameters
   parameterName=SunlightDirectionalVector
  }
  n2 = MaterialFunctionCall(Float4=@n1) {materialFunction=&BreakOutFloat4Components}
  n3 = DotProduct(
   A=Normalize(VectorInput=MaterialFunctionCall(X=@n2.R, Y=@n2.R, Z=@n2.R) {materialFunction=&MakeFloat3}.Result)
   B=PixelNormalWS()
  )
  n4 = PreSkinnedPosition()
  n5 = ShaderStageSwitch(PixelShader=VertexInterpolator(VS=@n4).PS, VertexShader=@n4)
  n6 = VectorParameter() {group=SandCastle, parameterName=GradientZ_Offset, sortPriority=500, defaultValue=<rgba 0 0 0 1>}
  n7 = ChannelMaskParameter(None=Divide(
    A=Add(A=@n5, B=@n6.RGB)
    B=MaterialFunctionCall() {materialFunction=&ObjectLocalBounds}.Local Bounds Size
   )) {group=SandCastle, maskChannel=Blue, parameterName=ZMask, sortPriority=500, defaultValue=<rgba 0 0 1 0>}
  Add(A=Saturate(None=@n3), B=@n7.R)
  n8 = Multiply(A=TextureCoordinate()) {constB=4}
  Add(A=@n8, B=Multiply(A=Time()) {constB=0.05})
  n9 = NamedRerouteUsage()
  Add(
   A=@n9
   B=TextureSample(UVs=@n8) {samplerType=SAMPLERTYPE_Normal, texture=/Game/Weapons/WeaponSkins/Wraps/Textures/T_Wrap_metallicFlakes_Small}.RGB
  )
  Constant() {r=0.5}
  n10 = Multiply(A=NamedRerouteUsage(), B=NamedRerouteUsage())
  n11 = Multiply(A=ScalarParameter() {defaultValue=0.25, group=SandCastle, parameterName=CM_Emissive}, B=@n10)
  n12 = Multiply(A=ScalarParameter() {defaultValue=2, group=SandCastle, parameterName=CM_EmissiveMax}, B=@n10)
  n13 = MaterialFunctionCall(Fade Length=Constant() {r=2500}, Fade Offset=Constant() {r=250}) {materialFunction=&CameraDepthFade}
  n14 = MakeMaterialAttributes!(
   BaseColor=NamedRerouteUsage()
   Metallic=ScalarParameter() {group=SandCastle, parameterName=CM_Metallic}
   Specular=NamedRerouteUsage()
   Roughness=NamedRerouteUsage()
   Anisotropy
   EmissiveColor=ShadingPathSwitch(Default=LinearInterpolate(A=@n11, B=@n12, Alpha=@n13.Result), Mobile=Constant())
   Opacity
   OpacityMask
   Normal=NamedRerouteUsage()
   Tangent
   WorldPositionOffset
   SubsurfaceColor
   ClearCoat
   ClearCoatRoughness
   AmbientOcclusion=NamedRerouteUsage()
   Refraction
   CustomizedUVs_0
   CustomizedUVs_1
   CustomizedUVs_2
   CustomizedUVs_3
   CustomizedUVs_4
   CustomizedUVs_5
   CustomizedUVs_6
   CustomizedUVs_7
   PixelDepthOffset
   ShadingModel
   Displacement
  )
  FunctionOutput(None=StaticSwitchParameter(
    True=MaterialFunctionCall!(In=@n14) {materialFunction=&/MF_PrincessCastle_SandFuzz}.Result
    False=NamedRerouteUsage()
   ) {group=Customization, parameterName=Use_SandCastle})
  n15 = LinearInterpolate(
   A=ScalarParameter() {defaultValue=0.33, group=SandCastle, parameterName=CM_SandCastle_RoughnessMin}
   B=ScalarParameter() {defaultValue=2, group=SandCastle, parameterName=CM_SandCastle_RoughnessMax}
   Alpha=NamedRerouteUsage()
  )
  LinearInterpolate(A=@n15, B=ScalarParameter() {defaultValue=0.44, group=SandCastle, parameterName=CM_SandCastle_RoughnessSparkle}, Alpha=NamedRerouteUsage())
  n16 = TextureObjectParameter() {
   group=SandCastle
   parameterName=CM_SandCastle_BaseColor
   sortPriority=0
   texture=/Game/Creative/Sets/PropSets/Primitives/Customization/Textures/SandCastle/T_CM_SandCastle_BC
  }
  n17 = ScalarParameter() {defaultValue=1, group=SandCastle, parameterName=CM_SandCastle_UVTile, sortPriority=100}
  n18 = NamedRerouteUsage()
  n19 = NamedRerouteUsage()
  n20 = Constant()
  n21 = ScalarParameter() {defaultValue=4, group=SandCastle, parameterName=CM_SandCastle_TriplannarContrast, sortPriority=101}
  n22 = NamedRerouteUsage()
  n23 = TextureObjectParameter() {
   group=SandCastle
   parameterName=CM_SandCastle_Normal
   samplerType=SAMPLERTYPE_Normal
   sortPriority=0
   texture=/Game/Creative/Sets/PropSets/Primitives/Customization/Textures/SandCastle/T_CM_SandCastle_N1
  }
  n24 = MaterialFunctionCall!(
   TextureObject_Diffuse=@n16
   TextureScale=@n17
   CalculatedBounds=@n18
   WorldNormals=@n19
   RotationX=@n20
   RotationY=@n20
   RotationZ=@n20
   ProjectionTransitionContrast=@n21
   Material Normals=@n22
   NormalTexture=@n23
   Normal Texture Intensity=ScalarParameter() {defaultValue=0.75, group=SandCastle, parameterName=CM_SandCastle_NormalIntensity, sortPriority=99}
  ) {materialFunction=/Game/Creative/Sets/PropSets/Primitives/Customization/MaterialFunctions/MF_Primitive_TriplanarUVs}
  n25 = ComponentMask(None=@n24.RGBA) {b=true, g=true, r=true}
  n26 = VectorParameter() {group=SandCastle, parameterName=CM_SandCastle_BaseColorTint, defaultValue=<rgba 0.8078 0.6471 0.5294 1>}
  n27 = MaterialFunctionCall(Base=@n25, Blend=@n26.RGB) {materialFunction=&Blend_Overlay}
  LinearInterpolate(A=@n27.Result, B=@n26.RGB)
  n28 = TextureObjectParameter() {
   group=SandCastle
   parameterName=CM_SandCastle_S
   sortPriority=0
   texture=/Game/Creative/Sets/PropSets/Primitives/Customization/Textures/SandCastle/T_CM_SandCastle_S
  }
  n29 = ScalarParameter() {defaultValue=12, group=SandCastle, parameterName=CM_SandCastle_NormalLargeTile, sortPriority=99}
  n30 = ScalarParameter() {defaultValue=4, group=SandCastle, parameterName=CM_SandCastle_NormalLargeContrast, sortPriority=99}
  n31 = TextureObjectParameter() {
   group=SandCastle
   parameterName=CM_SandCastle_NormalLarge
   samplerType=SAMPLERTYPE_Normal
   sortPriority=0
   texture=/Game/Creative/Sets/PropSets/Primitives/Customization/Textures/SandCastle/T_CM_SandCastle_N2
  }
  n32 = MaterialFunctionCall(
   TextureObject_Diffuse=@n28
   TextureScale=@n29
   CalculatedBounds=@n18
   WorldNormals=@n19
   RotationX=@n20
   RotationY=@n20
   RotationZ=@n20
   ProjectionTransitionContrast=@n30
   Material Normals=@n22
   NormalTexture=@n31
   Normal Texture Intensity=ScalarParameter() {defaultValue=1, group=SandCastle, parameterName=CM_SandCastle_NormalLargeIntensity, sortPriority=99}
  ) {materialFunction=&MF_Primitive_TriplanarUVs}
  n33 = ComponentMask(None=VertexNormalWS()) {b=true}
  n34 = OneMinus(None=@n33)
  LinearInterpolate(
   A=@n32.Normal Texture
   B=MaterialFunctionCall(Normal=@n32.Normal Texture, Flatness=ScalarParameter() {defaultValue=-2, group=SandCastle, parameterName=AdditionalNormalStrength}) {materialFunction=&FlattenNormal}.Result
   Alpha=Saturate(None=StaticSwitchParameter(True=@n34, False=@n33) {defaultValue=true, group=SandCastle, parameterName=AdditionalNormalStrengthZaxis, sortPriority=-100})
  )
  Multiply(A=@n1) {constB=-1}
  n35 = MaterialFunctionCall!(TextureObject=@n28, TextureSize=@n17, WorldPosition=@n18, World Space Normal=@n19, ProjectionTransitionContrast=@n21) {materialFunction=/Game/Creative/Sets/PropSets/Primitives/Customization/MaterialFunctions/MF_Primitive_TriplanarUVSimple}
  NamedRerouteDeclaration(None=ComponentMask(None=@n35.XYZ Texture) {r=true}) {name=Roughness, nodeColor=<rgba 0.5703 1 0 1>}
  n36 = GetMaterialAttributes(None=FunctionInput() {inputType=FunctionInput_MaterialAttributes}) {attributeGetTypes [#1, #2, #3]}
  NamedRerouteDeclaration(None=@n36.MaterialAttributes) {name=Material Attribute, nodeColor=<rgba 0.6641 0 1 1>}
  NamedRerouteDeclaration(None=@n26.RGB) {name=SparkleColor, nodeColor=<rgba 0 0.1328 1 1>}
  NamedRerouteDeclaration(None=@n13.Result) {name=CameraDepth, nodeColor=<rgba 1 0.4688 0 1>}
  NamedRerouteDeclaration(None=@n25) {name=Finalcolor, nodeColor=<rgba 0 0.5781 1 1>}
  NamedRerouteDeclaration(None=ComponentMask(None=@n35.XYZ Texture) {b=true}) {name=Specular, nodeColor=<rgba 0 0.6016 1 1>}
  NamedRerouteDeclaration(None=ComponentMask(None=MaterialFunctionCall(TextureObject=@n28, TextureSize=@n29, WorldPosition=@n18, World Space Normal=@n19, ProjectionTransitionContrast=@n30) {materialFunction=&MF_Primitive_TriplanarUVSimple}.XYZ Texture) {g=true}) {name=AO, nodeColor=<rgba 1 0.1172 0 1>}
  NamedRerouteDeclaration(None=NamedRerouteUsage()) {name=UVs, nodeColor=<rgba 0 1 0.4375 1>}
  n37 = Saturate(None=NamedRerouteUsage())
  n38 = NamedRerouteUsage()
  n39 = NamedRerouteUsage()
  NamedRerouteDeclaration(None=LinearInterpolate(A=LinearInterpolate(A=@n25, B=@n27.Result, Alpha=@n37), B=@n27.Result, Alpha=Max(A=@n38, B=@n39))) {name=BaseColor, nodeColor=<rgba 0 0.8594 1 1>}
  n40 = CameraVectorWS()
  n41 = Normalize(VectorInput=Transform(None=@n9))
  NamedRerouteDeclaration(None=Power(Base=DotProduct(A=@n40, B=@n41), Exp=ScalarParameter() {defaultValue=8, parameterName=Power})) {name=GeneralSparkle, nodeColor=<rgba 1 0 0.8906 1>}
  n42 = ScalarParameter() {defaultValue=0.5, group=SandCastle, parameterName=CM_SandCastle_AOMin}
  n43 = ScalarParameter() {defaultValue=1, group=SandCastle, parameterName=CM_SandCastle_AOMax}
  n44 = NamedRerouteUsage()
  n45 = NamedRerouteUsage()
  n46 = ScalarParameter() {defaultValue=1.25, group=SandCastle, parameterName=CM_SandCastle_AORoughnessContrast}
  n47 = Add(
   A=OneMinus(None=Power(Base=@n45, Exp=@n46))
   B=ScalarParameter() {defaultValue=0.6, group=SandCastle, parameterName=CM_SandCastle_AORoughnessAdd}
  )
  NamedRerouteDeclaration(None=Multiply(A=LinearInterpolate(A=@n42, B=@n43, Alpha=@n44), B=Saturate(None=@n47))) {name=FinalAO, nodeColor=<rgba 0 1 0.8828 1>}
  n48 = DotProduct(A=VertexNormalWS(), B=CameraVectorWS())
  NamedRerouteDeclaration(None=Power(Base=OneMinus(None=@n48))) {name=Fresnel, nodeColor=<rgba 0 1 0.7891 1>}
  n49 = TextureObjectParameter() {
   group=SandCastle
   parameterName=CM_SandCastle_Detail
   sortPriority=0
   texture=/Game/Creative/Sets/PropSets/Primitives/Customization/Textures/SandCastle/T_CM_SandCastle_M
  }
  n50 = ComponentMask(None=MaterialFunctionCall(
    TextureObject_Diffuse=@n49
    TextureScale=ScalarParameter() {defaultValue=0.7, group=SandCastle, parameterName=CM_SandCastle_NormalDetailTile, sortPriority=99}
    CalculatedBounds=@n18
    WorldNormals=@n19
    RotationX=@n20
    RotationY=@n20
    RotationZ=@n20
    ProjectionTransitionContrast=ScalarParameter() {defaultValue=4, group=SandCastle, parameterName=CM_SandCastle_NormalDetailContrast, sortPriority=99}
    Material Normals=@n22
    Normal Texture Intensity=ScalarParameter() {defaultValue=0.5, group=SandCastle, parameterName=CM_SandCastle_NormalDetailIntensity, sortPriority=99}
   ) {materialFunction=&MF_Primitive_TriplanarUVs}.RGBA) {b=true, g=true, r=true}
  n51 = MaterialFunctionCall(
   Hue Shift Percentage=Multiply(A=Time(), B=ScalarParameter() {defaultValue=0.175, group=SandCastle, parameterName=CM_SandCastle_SparkleSpeed, sortPriority=99}) {constB=0.075}
   Texture=Normalize(VectorInput=Subtract(A=@n50) {constB=0.5})
  ) {materialFunction=&HueShift}
  NamedRerouteDeclaration(None=@n51.Result) {name=Sparkle, nodeColor=<rgba 0.8984 0 1 1>}
  NamedRerouteDeclaration(None=@n36.AmbientOcclusion) {name=InAO, nodeColor=<rgba 0 1 0.7422 1>}
  NamedRerouteDeclaration(None=@n36.Roughness) {name=InRoughness, nodeColor=<rgba 0 0.6016 1 1>}
  NamedRerouteDeclaration(None=LinearInterpolate(
    A=@n36.Normal
    B=Constant3Vector() {constant=<rgba 0 0 1 0>}
    Alpha=ScalarParameter() {defaultValue=0.65, group=SandCastle, parameterName=CM_BaseNormalBlend}
   )) {name=InNormal, nodeColor=<rgba 0 0.2266 1 1>}
  n52 = ScalarParameter() {group=SandCastle, parameterName=CM_SandCastle_SpecularMin}
  n53 = ScalarParameter() {defaultValue=1, group=SandCastle, parameterName=CM_SandCastle_SpecularMax}
  n54 = NamedRerouteUsage()
  NamedRerouteDeclaration(None=LinearInterpolate(
    A=LinearInterpolate(A=@n52, B=@n53, Alpha=@n54)
    B=ScalarParameter() {defaultValue=0.75, group=SandCastle, parameterName=CM_SandCastle_SpecularSparkle}
    Alpha=NamedRerouteUsage()
   )) {name=FinalSpec, nodeColor=<rgba 0 0.4072 0.3968 1>}
  NamedRerouteDeclaration(None=@n15) {name=FinalRough, nodeColor=<rgba 0.4072 0.2831 0 1>}
  n55 = Desaturation(None=NamedRerouteUsage(), Fraction=Constant() {r=1})
  n56 = Round(None=Multiply(A=Power(Base=@n55) {constExponent=3}) {constB=15})
  n57 = NamedRerouteDeclaration(None=MaterialFunctionCall(BaseNormal=@n24.Normal Texture, AdditionalNormal=@n32.Normal Texture) {materialFunction=&BlendAngleCorrectedNormals}.Result) {name=Normal, nodeColor=<rgba 0 0.8594 1 1>}
  n58 = ComponentMask(None=@n57) {r=true}
  n59 = ComponentMask(None=@n57) {g=true}
  n60 = LinearInterpolate(Alpha=Add(A=Add(A=@n58, B=@n59), B=NamedRerouteUsage())) {constA=0.33}
  n61 = Subtract(A=WorldPosition().XYZ, B=Transform(None=CameraPositionWS()) {transformSourceType=TRANSFORMSOURCE_Camera})
  n62 = Normalize(VectorInput=@n61)
  n63 = NamedRerouteUsage()
  n64 = Subtract(
   A=DotProduct(A=@n51.Result, B=OneMinus(None=@n62))
   B=LinearInterpolate(B=ScalarParameter() {defaultValue=0.5, group=SandCastle, parameterName=CM_SandCastle_SparkleSubtract, sortPriority=99}, Alpha=@n63) {constA=1.1}
  )
  n65 = LinearInterpolate(
   A=ScalarParameter() {defaultValue=2.15, group=SandCastle, parameterName=CM_SandCastle_SparkleIntensityNear, sortPriority=99}
   B=ScalarParameter() {defaultValue=50, group=SandCastle, parameterName=CM_SandCastle_SparkleIntensityFar, sortPriority=99}
   Alpha=@n63
  ) {constA=2}
  n66 = NamedRerouteUsage()
  n67 = Saturate(None=@n3)
  n68 = Multiply(
   A=Power(
    Base=Multiply(A=@n64, B=@n65)
    Exp=ScalarParameter() {defaultValue=1, group=SandCastle, parameterName=CM_SandCastle_SparkleContrast, sortPriority=99}
   )
   B=Add(A=Add(A=@n66) {constB=0.05}, B=@n67)
  )
  n69 = Add(A=Power(Base=@n66) {constExponent=2.5}, B=@n67)
  NamedRerouteDeclaration(None=Multiply(A=Multiply(A=@n56, B=@n60), B=Multiply(A=@n68, B=@n69))) {name=Sparkles, nodeColor=<rgba 0.8281 1 0 1>}
  NamedRerouteDeclaration(None=ComponentMask(None=@n32.RGBA) {g=true}) {name=LargeNoise, nodeColor=<rgba 1 0 0.7969 1>}
  n70 = Power(Base=@n48) {constExponent=5}
  NamedRerouteDeclaration(None=Saturate(None=Add(A=@n70) {constB=0.1})) {name=FresnelInvert, nodeColor=<rgba 1 0.0938 0 1>}
  n71 = MaterialFunctionCall() {materialFunction=/Game/Creative/Sets/PropSets/Primitives/Customization/MaterialFunctions/MF_Primitive_CalculatedBounds}
  NamedRerouteDeclaration(None=Divide(A=@n71.WorldPosition, B=ScalarParameter() {defaultValue=100, group=SandCastle, parameterName=CM_SandCastle_WorldScale, sortPriority=100})) {name=CalculatedBounds, nodeColor=<rgba 1 0.1406 0 1>}
  NamedRerouteDeclaration(None=@n71.World Normals) {name=WorldNorm, nodeColor=<rgba 0.7344 0 1 1>}
  OneMinus(None=@n59)
  Subtract(A=@n61, B=MaterialFunctionCall() {materialFunction=&CameraDirectionVector}.Result)
  TransformPosition(None=@n71.WorldPosition) {transformSourceType=TRANSFORMPOSSOURCE_World}
 ]}
